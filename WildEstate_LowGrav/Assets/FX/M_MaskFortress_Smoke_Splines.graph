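# Translucent, unlit, two-sided material for Mask Fortress spline smoke effects on Niagara mesh particles.
# SplineThicken builds ribbon width and world position offset, with reroutes scaling the width.
# A panned plasma texture, offset by vertex alpha, drives HDR emissive and gradient-masked, Fresnel-shaped opacity.
# Named reroutes supply curl noise, near and far distance fades, a sine-driven moving mask and sphere-mask glow, alongside a fake fog color.
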Material /WildEstate_LowGrav/Assets/FX/M_MaskFortress_Smoke_Splines
graph {
 properties {
  bAutomaticallySetUsageInEditor=false
  bUsedWithNiagaraMeshParticles=true
  blendMode=BLEND_Translucent
  shadingModel=MSM_Unlit
  translucencyLightingMode=TLM_VolumetricPerVertexNonDirectional
  twoSided=true
  usedShadingModels=Unlit
 }
 nodes [
  n1 = Clamp(None=LinearInterpolate(Alpha=NamedRerouteUsage()) {constA=1, constB=0})
  n2 = NamedRerouteUsage()
  n3 = Clamp(None=Multiply(A=NamedRerouteUsage(), B=Multiply(A=@n2)))
  n4 = Multiply(A=Multiply(A=ScalarParameter() {defaultValue=100, parameterName="WidthBase (S)"}, B=@n1), B=@n3)
  n5 = Multiply(A=Multiply(A=ScalarParameter() {defaultValue=100, parameterName="WidthTip (S)"}, B=@n1), B=@n3)
  n6 = VertexNormalWS()
  Add(
   A=Add(
    A=MaterialFunctionCall(WidthBase=@n4, WidthTip=@n5) {materialFunction=&SplineThicken}.WorldPositionOffset
    B=@n6
   )
   B=VectorParameter() {parameterName=MoveOffGround, defaultValue=<rgba 0 0 0 1>}.RGB
  )
  n7 = Multiply(
   A=VectorParameter() {parameterName=FogColor, defaultValue=<rgba 0 0 0 1>}.RGB
   B=ScalarParameter() {defaultValue=1, parameterName=FakeFogAmount}
  )
  n8 = NamedRerouteUsage()
  n9 = TextureCoordinate()
  n10 = Multiply(A=NamedRerouteUsage(), B=OneMinus(None=ComponentMask(None=@n9) {g=true}))
  n11 = MaterialFunctionCall(
   LDR Input=@n10
   HDR Intensity Multiplier=Multiply(A=@n10, B=ScalarParameter() {defaultValue=25, parameterName="HDR Intensity Multiplier (S)"})
  ) {materialFunction=&DeriveHDRfromLDR}
  n12 = TextureCoordinate()
  n13 = AppendVector(A=ComponentMask(None=@n12) {r=true}, B=Add(A=ComponentMask(None=@n12) {g=true}, B=VertexColor().A))
  n14 = VectorParameter() {parameterName=PannerUVMult, defaultValue=<rgba 0.25 0.5 0 1>}
  n15 = AppendVector(A=ComponentMask(None=@n14.RGB) {r=true}, B=ComponentMask(None=@n14.RGB) {g=true})
  n16 = Constant() {r=-0.2}
  n17 = ScalarParameter() {defaultValue=0.5, parameterName=SpeedMultVertex}
  n18 = TextureSample(UVs=Panner(Coordinate=Multiply(A=@n13, B=@n15), Speed=AppendVector(A=@n16, B=@n17)) {speedY=1}) {samplerType=SAMPLERTYPE_LinearGrayscale, texture=/Game/Effects/Fort_Effects/Emotes/UltraLight/Texture/T_Plasma_A_B_Greyscale}
  n19 = ScalarParameter() {defaultValue=1, parameterName=Emissive}
  n20 = NamedRerouteUsage()
  n21 = Multiply(A=Multiply(A=Multiply(A=@n11.RemappedRange, B=@n18.G), B=Multiply(A=@n19, B=@n20)), B=NamedRerouteUsage())
  n22 = OneMinus(None=ComponentMask(None=TextureCoordinate()) {g=true})
  MaterialFunctionCall(Emissive=Add(A=LinearInterpolate(B=@n7, Alpha=@n8), B=Multiply(A=@n21, B=@n22))) {materialFunction=&MF_FX_EmissiveAdjustments}
  Multiply(A=@n6, B=NamedRerouteUsage())
  n23 = ScalarParameter() {defaultValue=1, parameterName=MoveMask_Speed}
  Multiply(A=@n23, B=LinearInterpolate(Alpha=VertexColor().A) {constA=1, constB=1.25})
  n24 = Multiply(A=NamedRerouteUsage(), B=NamedRerouteUsage())
  n25 = ScalarParameter() {defaultValue=5, parameterName=FinalOpacity2}
  n26 = Constant() {r=0.25}
  Multiply(A=Multiply(A=Clamp(None=@n24), B=Constant() {r=1}), B=LinearInterpolate(A=Multiply(A=@n25, B=@n26), B=@n25, Alpha=NamedRerouteUsage()))
  NamedRerouteDeclaration(None=ScalarParameter() {defaultValue=1, parameterName=DistanceValueMultiplier}) {name=DistanceValueMultiplier, nodeColor=<rgba 1 0.6797 0 1>}
  n27 = WorldPosition()
  n28 = VectorParameter() {parameterName=SphereLocation, defaultValue=<rgba 0 0 0 1>}
  n29 = ScalarParameter() {defaultValue=500, parameterName="Radius (S)"}
  n30 = ScalarParameter() {parameterName=Hardness}
  NamedRerouteDeclaration(None=Multiply(A=ScalarParameter() {defaultValue=1, parameterName=SphereEmissive}, B=SphereMask(A=@n27.XYZ, B=@n28.RGB, Radius=@n29, Hardness=@n30)) {constB=250}) {name=SphereMask_Glow, nodeColor=<rgba 0.1484 1 0 1>}
  n31 = ScalarParameter() {defaultValue=2500, parameterName=FarFadeout_Distance}
  n32 = ScalarParameter() {defaultValue=-2500, parameterName=FarFadeout_Distance_Offset}
  NamedRerouteDeclaration(None=OneMinus(None=MaterialFunctionCall!(Blend Range=@n31, Start Offset=@n32) {materialFunction=/Engine/Functions/Engine_MaterialFunctions03/Particles/Distance_Blend}.Result)) {name=DistanceFade, nodeColor=<rgba 0.9155 0.651 1 1>}
  n33 = Time() {bIgnorePause=true, bOverride_Period=true, period=1000}
  n34 = Multiply(A=VectorParameter() {parameterName=PanSpeed, defaultValue=<rgba 250 1500 -500 1>}.RGB, B=NamedRerouteUsage())
  n35 = Multiply(
   A=VectorParameter() {parameterName=CurlNoise_Tiling, defaultValue=<rgba 0.02 0.02 0.02 1>}.RGB
   B=Add(A=Multiply(A=@n33, B=@n34) {constB=250}, B=WorldPosition().XYZ)
  )
  NamedRerouteDeclaration(None=Multiply(A=VectorNoise(World Position=@n35) {noiseFunction=VNF_CurlALU}, B=ScalarParameter() {defaultValue=1, parameterName=WPOCurlAmount})) {name=VectorNoise, nodeColor=<rgba 0.3153 0.9234 1 1>}
  NamedRerouteDeclaration(None=MaterialFunctionCall(
    Blend Range=ScalarParameter() {defaultValue=500, parameterName=CloseFadeout_Distance}
    Start Offset=ScalarParameter() {defaultValue=-1, parameterName=CloseFadeout_Distance_Offset}
   ) {materialFunction=&Distance_Blend}.Result) {name=Close_DistanceFade, nodeColor=<rgba 0.8285 0.2917 1 1>}
  n36 = Time()
  n37 = TextureCoordinate()
  n38 = Sine(None=Add(A=Multiply(A=@n23, B=@n36), B=ComponentMask(None=@n37) {g=true})) {period=2}
  n39 = Multiply(A=Add(A=@n38) {constB=1.5}) {constB=0.2}
  NamedRerouteDeclaration(None=SmoothStep(Value=Clamp(None=@n39)) {constMin=-0.5}) {name=Moving Gradient Mask Alpha, nodeColor=<rgba 1 0.6328 0 1>}
  n40 = NamedRerouteUsage()
  n41 = MaterialFunctionCall(UV Channel=TextureCoordinate() {vTiling=0.5}) {materialFunction=&LinearGradient}
  n42 = Multiply(A=Multiply(A=@n41.UGradient, B=OneMinus(None=@n41.UGradient)), B=Multiply(A=@n41.VGradient, B=OneMinus(None=@n41.VGradient)))
  n43 = Multiply(
   A=Multiply(A=Multiply(A=@n42, B=@n18.G), B=ScalarParameter() {defaultValue=1, parameterName=FinalOpacity})
   B=MaterialFunctionCall(Exponent=ScalarParameter() {defaultValue=1, parameterName="Exponent (S)"}) {materialFunction=&AxisAlignedFresnel}.Fresnel
  )
  NamedRerouteDeclaration(None=Multiply(A=NamedRerouteUsage(), B=Multiply(A=@n40, B=@n43))) {name=DistanceFading_Complete, nodeColor=<rgba 0.8513 0.4325 1 1>}
  OneMinus(None=ConstantBiasScale(None=@n38) {bias=0.5, scale=1})
 ]
}
