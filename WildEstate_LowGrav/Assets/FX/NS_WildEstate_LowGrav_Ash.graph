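# Niagara system for ambient floating ash in the WildEstate LowGrav map, using one GPU emitter; Screen_Dust is disabled.
# User.Spawn sets spawn probability for particles partitioned three ways, each with its own color, size, forces and visibility tag for three sprite renderers.
# Particles spawn in a squashed sphere, then move under upward gravity, vortex, curl noise and drag.
# Vortex origin and two kill spheres use hardcoded coordinates, tying the effect to one level layout.
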NiagaraSystem /WildEstate_LowGrav/Assets/FX/NS_WildEstate_LowGrav_Ash
dependencies {
 usedDataInterfaces [
  &NiagaraDataInterfaceSpriteRendererInfo
  &NiagaraDataInterfaceCurve
  &NiagaraDataInterfaceCamera
  &NiagaraDataInterfaceCollisionQuery
  &NiagaraDataInterfaceLandscape
 ]
 usedDynamicInputs []
 usedModules [&SystemState]
}
emitters {
 Ash {
  simTarget=GPUComputeSim
  emitterUpdateScript [
   &EmitterState()
   &SpawnRate(Spawn Probability=$User.Spawn, SpawnRate=220, Use Spawn Probability=true)
  ]
  particleSpawnScript [
   &PartitionParticles(Number Of Partitions=3)
   /SetVariables(Particles.VisibilityTag=$Particles.PartitionParticles.Partition) {isSetParameters=true}
   &InitializeParticle(
    Color=&MergeHelper_ColorSwitch(
     0=<rgba 5 4.9603 4.5699 1>
     1=<rgba 0.0104 0.2865 0.0355 1>
     2=<rgba 0.0881 0.099 0.0902 1>
     3=?
     Index=$Particles.PartitionParticles.Partition
     Num=::NewEnumerator1
    )
    Lifetime=4
    Sprite Rotation Mode=::NewEnumerator2
    Sprite Size Mode=::NewEnumerator3
    Sprite UV Mode=::NewEnumerator3
    Uniform Sprite Size=&MergeHelper_FloatSwitch(
     0=&RandomRangeFloat(Minimum=20, Maximum=25)
     1=&RandomRangeFloat(Minimum=35, Maximum=60)
     2=&RandomRangeFloat(Minimum=8, Maximum=11)
     3=?
     Index=$Particles.PartitionParticles.Partition
     Num=::NewEnumerator1
    )
    UsePositionOffset=false
    Write Material Random=true
   )
   &ShapeLocation(
    MaskSpawnGroup=false
    Non Uniform Scale=&MergeHelper_VectorSwitch(0=<1.1 1.1 0.8>, 1=<1.1 1.1 0.8>, 2=<1.1 1.1 0.2>, 3=?, Index=$Particles.PartitionParticles.Partition, Num=::NewEnumerator1)
    Offset=&MergeHelper_VectorSwitch(0=<0 0 -350>, 1=<0 0 -350>, 2=<0 0 -500>, 3=?, Index=$Particles.PartitionParticles.Partition, Num=::NewEnumerator1)
    Offset Coordinate Space=::World
    Offset Mode=::NewEnumerator0
    Override Seed=false
    Sphere Radius=2800
    UseOwnerScale=true
   )
   &SubUVAnimation(
    Evaluation Type=::NewEnumerator0
    "Sprite Renderer "=DataInterface() {propertyValues {SpriteRenderer=&self:Ash_0.NiagaraSpriteRendererProperties_0}}
    SubUV Animation Mode:ENiagara_SubUVLookupModeV2=::NewEnumerator2
    UseEndFrame=false
    UseStartFrame=false
   )
   &SubUVAnimation(
    Evaluation Type=::NewEnumerator0
    "Sprite Renderer "=DataInterface() {propertyValues {SpriteRenderer=&self:Ash_0.NiagaraSpriteRendererProperties_2}}
    SubUV Animation Mode:ENiagara_SubUVLookupModeV2=::NewEnumerator2
    UseEndFrame=false
    UseStartFrame=false
   ) {module=SubUVAnimation002}
   &DynamicMaterialParameters(Module.Index0Float0Write=true, Module.Index0Float1Write=true, Module.Index0Float2Write=true, Module.Index0Float3Write=true) {enabled=false}
  ]
  particleUpdateScript [
   &ParticleState()
   &ScaleSpriteSize(
    Uniform Curve Scale=/Niagara/DynamicInputs/Camera/CameraFade(
     Camera Query:NiagaraDataInterfaceCamera=DataInterface()
     Fade Offset:NiagaraFloat=100
     Fade Distance:NiagaraFloat=200
     Use Owner Position:NiagaraBool=true
    )
    Uniform Curve Sprite Scale=DataInterface() {propertyValues {Curve {keys [
        {interpMode=RCIM_Cubic, tangentMode=RCTM_Break, time=0, value=0}
        {interpMode=RCIM_Cubic, tangentMode=RCTM_Break, time=0.5, value=1}
        {interpMode=RCIM_Cubic, tangentMode=RCTM_Break, time=1, value=0}
       ]}}}
   )
   &SpriteRotationRate(Rotation Rate=90)
   &WindForce(Wind Speed=<500 0 0>) {enabled=false}
   &VortexForce(
    Origin Pull Only On Vortex Axis=false
    UseOriginOffset=false
    Vortex Force Amount=&MergeHelper_FloatSwitch(0=1111, 1=1111, 2=250, 3=?, Index=$Particles.PartitionParticles.Partition, Num=::NewEnumerator1)
    Vortex Origin:NiagaraPosition=&AddVectorToPosition(Position=<49048.7461 -45291.9219 3216.2664>, Vector=<0 0 0>)
   )
   &GravityForce(Gravity=&MergeHelper_VectorSwitch(
     0=&RandomRangeVector(Minimum=<0 0 0>, Maximum=<0 0 600>)
     1=&RandomRangeVector(Minimum=<0 0 0>, Maximum=<1 1 700>)
     2=&RandomRangeVector(Minimum=<0 0 0>, Maximum=<0 0 100>)
     3=?
     Index=$Particles.PartitionParticles.Partition
     Num=::NewEnumerator1
    ))
   &Drag(Use Linear Drag=true, Use Rotational Drag=true)
   &AerodynamicDrag(
    Aerodynamic Drag:NiagaraFloat=&MergeHelper_FloatSwitch(0=0.33, 1=0.33, 2=0.1, 3=?, Index=$Particles.PartitionParticles.Partition, Num=::NewEnumerator1)
    Aerodynamic Rotational Drag=0
    Pivot Offset Maximum:Vector3f=<0.2 0.25 0.1>
    Pivot Offset Minimum:Vector3f=<-0.2 -0.1 -0.1>
    Pivot Offset Minimum Length:Vector3f=<0.1 0 0>
    Rotation Strength=0
    Use Linear Drag=true
    Use Rotational Drag=true
   )
   &CurlNoiseForce(
    Bias Noise=false
    Noise Frequency=20
    Noise Strength=&MergeHelper_FloatSwitch(0=5000, 1=5000, 2=0, 3=?, Index=$Particles.PartitionParticles.Partition, Num=::NewEnumerator1)
    Pan Noise=true
    Pan Noise Field=<0 0 1>
    Randomization Vector:Vector3f=<0.65 0.3 0.1>
   )
   &SolveForcesAndVelocity(Clamp Velocity=false, Limit Acceleration=false)
   &AlignSpriteToMeshOrientation(SetAlignment=true, SetFacing=true)
   &KillParticlesInVolume(Sphere Radius=2010, Volume Origin:Vector3f=<45132 -51208 9300>)
   &KillParticlesInVolume(Sphere Radius=2010, Volume Origin:Vector3f=<54352 -51208 9300>) {module=KillParticlesInVolume001}
   &SubUVAnimation(
    Play Rate:NiagaraFloat=&MergeHelper_FloatSwitch(0=1, 1=1, 2=0, 3=?, Index=$Particles.PartitionParticles.Partition, Num=::NewEnumerator1)
    Random Start Frame=true
    "Sprite Renderer "=DataInterface() {propertyValues {SpriteRenderer=&self:Ash_0.NiagaraSpriteRendererProperties_1}}
    SubUV Animation Mode:ENiagara_SubUVLookupModeV2=::NewEnumerator3
    UseEndFrame=false
    UsePlayRateScale=false
    UseStartFrame=false
   ) {module=SubUVAnimation001}
  ]
  rendererClasses [&NiagaraSpriteRendererProperties]
  renderers [
   {rendererClass=&NiagaraSpriteRendererProperties}
   {rendererClass=&NiagaraSpriteRendererProperties, rendererIndex=1}
   {rendererClass=&NiagaraSpriteRendererProperties, rendererIndex=2}
  ]
 }
 Screen_Dust {bEnabled=false, rendererClasses [], renderers []}
}
summary {
 systemName=NS_WildEstate_LowGrav_Ash
 emitters [
  {emitterName=Ash, simTarget=GPUComputeSim, rendererClasses [&NiagaraSpriteRendererProperties]}
  {bEnabled=false, emitterName=Screen_Dust, rendererClasses []}
 ]
 userVariables [User.Spawn {type=NiagaraFloat, defaultValue=1, description=""}]
}
